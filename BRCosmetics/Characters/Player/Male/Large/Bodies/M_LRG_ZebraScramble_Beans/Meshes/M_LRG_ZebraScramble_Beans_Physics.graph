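# Physics asset for the ZebraScramble Beans large male player body, setting up jacket physics around the lower torso and thighs.
# Capsules on pelvis, spine_01 and both thighs form the body shapes.
# Seven jacket chains (front, side and rear per side, plus center rear) each run from a tiny root through upper and lower segments.
# Roots and thighs have free angular joints, while jacket segments use tight limits, with wider swing at the front.
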PhysicsAsset /BRCosmetics/Characters/Player/Male/Large/Bodies/M_LRG_ZebraScramble_Beans/Meshes/M_LRG_ZebraScramble_Beans_Physics
bodies {
 dyn_jacket_l_front [{shapeType=Capsule, length=14.5, radius=5.125, shapeName=dyn_jacket_l_front_capsule, center=<4.2072 -3.6961 0>}]
 dyn_jacket_l_front_lower [{
   shapeType=Capsule
   length=10.4148
   radius=7.2623
   shapeName=dyn_jacket_l_front_lower_capsule
   center=<7.1473 -4.4827 0.455>
   rotation=<rot 7.3699 -90 -2.654>
  }]
 dyn_jacket_l_front_root [{shapeType=Capsule, length=0.505, radius=0.505, shapeName=dyn_jacket_l_front_root_capsule, rotation=<rot 0 -90 0>}]
 dyn_jacket_l_rear [{shapeType=Capsule, length=10.5625, radius=4.5625, shapeName=dyn_jacket_r_rear_capsule, center=<2.9101 -3.59 0>, rotation=<rot 0 0 -180>}]
 dyn_jacket_l_rear_lower [{
   shapeType=Capsule
   length=17.3438
   radius=4.7681
   shapeName=dyn_jacket_l_rear_lower_capsule
   center=<6.3199 -1.3902 0>
   rotation=<rot 0 -90 0>
  }]
 dyn_jacket_l_rear_root [{shapeType=Capsule, length=0.505, radius=0.505, shapeName=dyn_jacket_l_rear_root_capsule, rotation=<rot 0 -90 0>}]
 dyn_jacket_l_side [{shapeType=Capsule, length=14.5, radius=4.5, shapeName=dyn_jacket_r_side_capsule, center=<3.446 -5.2302 0>, rotation=<rot 0 0 -180>}]
 dyn_jacket_l_side_lower [{
   shapeType=Capsule
   length=11.3959
   radius=6.6121
   shapeName=dyn_jacket_l_side_lower_capsule
   center=<6.8422 -4.7087 -0.011>
   rotation=<rot -3.0694 -90.6169 0>
  }]
 dyn_jacket_l_side_root [{shapeType=Capsule, length=0.505, radius=0.505, shapeName=dyn_jacket_l_side_root_capsule, rotation=<rot 0 -90 0>}]
 dyn_jacket_r_front [{shapeType=Capsule, length=14.5, radius=5.125, shapeName=dyn_jacket_l_front_capsule, center=<-4.2072 3.6961 0>, rotation=<rot 0 0 -180>}]
 dyn_jacket_r_front_lower [{
   shapeType=Capsule
   length=10.4148
   radius=7.2623
   shapeName=dyn_jacket_l_front_lower_capsule
   center=<-7.1473 4.4827 -0.455>
   rotation=<rot 7.3699 -90 177.346>
  }]
 dyn_jacket_r_front_root [{shapeType=Capsule, length=0.505, radius=0.505, shapeName=dyn_jacket_r_front_root_capsule, rotation=<rot 0 -90 0>}]
 dyn_jacket_r_rear [{shapeType=Capsule, length=10.5625, radius=4.5625, shapeName=dyn_jacket_r_rear_capsule, center=<-2.9101 3.59 0>}]
 dyn_jacket_r_rear_lower [{
   shapeType=Capsule
   length=17.3438
   radius=4.7681
   shapeName=dyn_jacket_l_rear_lower_capsule
   center=<-6.3199 1.3902 0>
   rotation=<rot 0 -90 180>
  }]
 dyn_jacket_r_rear_root [{shapeType=Capsule, length=0.505, radius=0.505, shapeName=dyn_jacket_r_rear_root_capsule, rotation=<rot 0 -90 0>}]
 dyn_jacket_r_side [{shapeType=Capsule, length=14.5, radius=4.5, shapeName=dyn_jacket_r_side_capsule, center=<-3.446 5.2302 0>}]
 dyn_jacket_r_side_lower [{
   shapeType=Capsule
   length=11.3959
   radius=6.6121
   shapeName=dyn_jacket_l_side_lower_capsule
   center=<-6.8422 4.7087 0.011>
   rotation=<rot -3.0694 -90.6169 -180>
  }]
 dyn_jacket_r_side_root [{shapeType=Capsule, length=0.505, radius=0.505, shapeName=dyn_jacket_r_side_root_capsule, rotation=<rot 0 -90 0>}]
 dyn_jacket_rear [{shapeType=Box, extentX=12, extentY=6.5, extentZ=24.5, shapeName=dyn_jacket_rear_box, center=<4.4593 -3.5508 0>, rotation=<rot 0 -0.7614 0>}]
 dyn_jacket_rear_lower [{shapeType=Capsule, length=14.8745, radius=4.8788, shapeName=dyn_jacket_rear_lower_capsule, center=<7.0343 -2.1264 0>, rotation=<rot 0 -90 0>}]
 dyn_jacket_rear_root [{shapeType=Capsule, length=0.505, radius=0.505, shapeName=dyn_jacket_rear_root_capsule, rotation=<rot 0 -90 0>}]
 pelvis [{shapeType=Capsule, length=10.1058, radius=17.3847, center=<-1.1244 1.1843 0.0088>, rotation=<rot 0.0148 -90.6849 -0.0793>}]
 spine_01 [{shapeType=Capsule, length=6.3478, radius=18.608, center=<1.6917 1.1806 -0.0086>, rotation=<rot 1.4969 -90 0.0015>}]
 thigh_l [{
   shapeType=Capsule
   length=36.5183
   radius=12.193
   shapeName=thigh_l_capsule
   center=<-21.9413 1.5921 2.0753>
   rotation=<rot 9.3969 -90.0001 88.0582>
  }]
 thigh_r [{
   shapeType=Capsule
   length=36.5183
   radius=12.193
   shapeName=thigh_l_capsule
   center=<21.9413 -1.5921 -2.0753>
   rotation=<rot 9.3969 -90.0001 -91.9418>
  }]
}
constraints [
 {bone1Name=spine_01, bone2Name=pelvis, swing1LimitDegrees=45, swing2LimitDegrees=45, twistLimitDegrees=45}
 {
  bone1Name=dyn_jacket_l_front_root
  bone2Name=spine_01
  swing1LimitDegrees=45
  swing1Motion=Free
  swing2LimitDegrees=45
  swing2Motion=Free
  twistLimitDegrees=45
  twistMotion=Free
 }
 {
  bone1Name=dyn_jacket_l_rear_root
  bone2Name=spine_01
  swing1LimitDegrees=45
  swing1Motion=Free
  swing2LimitDegrees=45
  swing2Motion=Free
  twistLimitDegrees=45
  twistMotion=Free
 }
 {
  bone1Name=dyn_jacket_l_side_root
  bone2Name=spine_01
  swing1LimitDegrees=45
  swing1Motion=Free
  swing2LimitDegrees=45
  swing2Motion=Free
  twistLimitDegrees=45
  twistMotion=Free
 }
 {
  bone1Name=dyn_jacket_r_front_root
  bone2Name=spine_01
  swing1LimitDegrees=45
  swing1Motion=Free
  swing2LimitDegrees=45
  swing2Motion=Free
  twistLimitDegrees=45
  twistMotion=Free
 }
 {
  bone1Name=dyn_jacket_r_rear_root
  bone2Name=spine_01
  swing1LimitDegrees=45
  swing1Motion=Free
  swing2LimitDegrees=45
  swing2Motion=Free
  twistLimitDegrees=45
  twistMotion=Free
 }
 {
  bone1Name=dyn_jacket_r_side_root
  bone2Name=spine_01
  swing1LimitDegrees=45
  swing1Motion=Free
  swing2LimitDegrees=45
  swing2Motion=Free
  twistLimitDegrees=45
  twistMotion=Free
 }
 {
  bone1Name=dyn_jacket_rear_root
  bone2Name=spine_01
  swing1LimitDegrees=45
  swing1Motion=Free
  swing2LimitDegrees=45
  swing2Motion=Free
  twistLimitDegrees=45
  twistMotion=Free
 }
 {bone1Name=dyn_jacket_rear, bone2Name=dyn_jacket_rear_root, swing1LimitDegrees=24, swing2LimitDegrees=8, twistLimitDegrees=8}
 {bone1Name=dyn_jacket_l_front_lower, bone2Name=dyn_jacket_l_front, swing1LimitDegrees=45, swing2LimitDegrees=8, twistLimitDegrees=8}
 {bone1Name=dyn_jacket_rear_lower, bone2Name=dyn_jacket_rear, swing1LimitDegrees=24, swing2LimitDegrees=8, twistLimitDegrees=8}
 {
  bone1Name=thigh_l
  bone2Name=pelvis
  swing1LimitDegrees=45
  swing1Motion=Free
  swing2LimitDegrees=45
  swing2Motion=Free
  twistLimitDegrees=45
  twistMotion=Free
 }
 {bone1Name=dyn_jacket_r_side, bone2Name=dyn_jacket_r_side_root, swing1LimitDegrees=24, swing2LimitDegrees=8, twistLimitDegrees=8}
 {bone1Name=dyn_jacket_r_side_lower, bone2Name=dyn_jacket_r_side, swing1LimitDegrees=24, swing2LimitDegrees=8, twistLimitDegrees=8}
 {
  bone1Name=thigh_r
  bone2Name=pelvis
  swing1LimitDegrees=45
  swing1Motion=Free
  swing2LimitDegrees=45
  swing2Motion=Free
  twistLimitDegrees=45
  twistMotion=Free
 }
 {bone1Name=dyn_jacket_r_rear_lower, bone2Name=dyn_jacket_r_rear, swing1LimitDegrees=24, swing2LimitDegrees=8, twistLimitDegrees=8}
 {bone1Name=dyn_jacket_r_rear, bone2Name=dyn_jacket_r_rear_root, swing1LimitDegrees=24, swing2LimitDegrees=8, twistLimitDegrees=8}
 {bone1Name=dyn_jacket_l_side_lower, bone2Name=dyn_jacket_l_side, swing1LimitDegrees=24, swing2LimitDegrees=8, twistLimitDegrees=8}
 {bone1Name=dyn_jacket_l_side, bone2Name=dyn_jacket_l_side_root, swing1LimitDegrees=24, swing2LimitDegrees=8, twistLimitDegrees=8}
 {bone1Name=dyn_jacket_l_rear, bone2Name=dyn_jacket_l_rear_root, swing1LimitDegrees=24, swing2LimitDegrees=8, twistLimitDegrees=8}
 {bone1Name=dyn_jacket_l_rear_lower, bone2Name=dyn_jacket_l_rear, swing1LimitDegrees=24, swing2LimitDegrees=8, twistLimitDegrees=8}
 {bone1Name=dyn_jacket_r_front_lower, bone2Name=dyn_jacket_r_front, swing1LimitDegrees=45, swing2LimitDegrees=8, twistLimitDegrees=8}
 {bone1Name=dyn_jacket_r_front, bone2Name=dyn_jacket_r_front_root, swing1LimitDegrees=45, swing2LimitDegrees=8, twistLimitDegrees=8}
 {bone1Name=dyn_jacket_l_front, bone2Name=dyn_jacket_l_front_root, swing1LimitDegrees=45, swing2LimitDegrees=8, twistLimitDegrees=8}
]
