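# Physics asset of cape colliders for the large male ZebraScramble Beans player body, apparently the collision volume that keeps a cape off the character.
# Capsules cover the pelvis, spine, neck and clavicles, with a sphere on the head and extra capsules on the three rear jacket dynamic bones.
# Arm, leg and foot bodies have no shapes, and there are no constraints, so it defines collision rather than a jointed ragdoll.
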PhysicsAsset /BRCosmetics/Characters/Player/Male/Large/Bodies/M_LRG_ZebraScramble_Beans/Meshes/M_LRG_ZebraScramble_Beans_CapeColliders
bodies {
 calf_l []
 calf_r []
 clavicle_l [{shapeType=Capsule, length=16.0625, radius=16.1371, center=<9.8737 -0.8565 -5.2812>, rotation=<rot -89.6026 179.9874 -179.9874>}]
 clavicle_r [{shapeType=Capsule, length=16.0625, radius=16.1371, center=<-9.8737 0.8565 5.2812>, rotation=<rot -89.6026 179.9874 0.0126>}]
 dyn_jacket_l_rear [{
   shapeType=Capsule
   length=3.7129
   radius=8.2281
   shapeName=dyn_jacket_l_rear_capsule
   center=<9.879 2.9284 -0.2019>
   rotation=<rot 9.5264 -90 -1.9618>
  }]
 dyn_jacket_r_rear [{
   shapeType=Capsule
   length=3.7129
   radius=8.2281
   shapeName=dyn_jacket_l_rear_capsule
   center=<-9.879 -2.9284 0.2019>
   rotation=<rot 9.5264 -90 178.0382>
  }]
 dyn_jacket_rear [{shapeType=Capsule, length=7.789, radius=8.5679, shapeName=dyn_jacket_rear_capsule, center=<10.3548 -0.4825 0>, rotation=<rot 0 -90 0>}]
 foot_l []
 foot_r []
 head [{shapeType=Sphere, radius=17.3125, center=<11.028 -0.174 0>}]
 lowerarm_l []
 lowerarm_r []
 neck_02 [{shapeType=Capsule, length=13.4375, radius=7.3175, center=<-0.2642 -0.3408 0>, rotation=<rot 90 -171.1193 -175.3978>}]
 pelvis [{shapeType=Capsule, length=9.0206, radius=21.2263, center=<-3.197 0.7838 0>}]
 spine_01 [{shapeType=Capsule, length=5.3813, radius=22.0586, center=<-1.7965 1.1778 0>}]
 spine_02 [{shapeType=Capsule, length=6.5139, radius=22.2572, center=<-1.711 1.6789 0.0001>}]
 spine_03 [{shapeType=Capsule, length=12.8982, radius=18.5781, center=<0.1445 0.5544 0.0001>}]
 spine_04 [{shapeType=Capsule, length=31.5625, radius=17.5075, center=<2.766 -1.5254 0.0002>}]
 spine_05 [{shapeType=Capsule, length=24.0226, radius=21.195, center=<-3.5227 1.5654 0>}]
 thigh_l []
 thigh_r []
 upperarm_l []
 upperarm_r []
}
constraints []
